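# Sound cue for the third-person (3P) music of the GoatDance emote in Athena.
# A looping wave player feeds an enveloper whose volume curve gives a very short fade-in from silence.
# Emote attenuation, a 3P music sound class and property sheet, and shared emote-music concurrency govern playback, which continues when silent.
# It also has a submix send to the music audio-analysis submix.
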SoundCue /BRCosmetics/Sounds/Emotes/GoatDance/SC_Athena_Emote_GoatDance_Music_3P_Cue
root=Enveloper_0
nodes [
 {
  id=Enveloper_0
  children [WavePlayer_0]
  properties {pitchCurve {editorCurveData {keys [{time=0}]}}, volumeCurve {editorCurveData {keys [<at -0.0011 0.0001>,<at 0.0495 1>]}}}
 }
 {id=WavePlayer_0, properties {bLooping=true, soundWaveAssetPtr=&/Emote_GoatDance}}
]
properties {
 attenuationSettings=&Fort_Emotes_Athena_Att
 audioPropertiesSheet=/BRCosmetics/Sounds/Settings/PropertySheets/Emotes/APS_Cosmetic_Emote_Music_3P
 maxDistance=1600
 soundClassObject=/Game/Sounds/FortSoundClassesAndModes/Fort_Music_Emote_3P
 virtualizationMode=PlayWhenSilent
 volumeMultiplier=1
 concurrencyOverrides {resolutionRule=PreventNew}
 concurrencySet [/Game/Sounds/Concurrency_Rules/Athena_EmoteMusic_Conc]
 soundSubmixSends [{minSendLevel=0, soundSubmix=&Fortnite_AudioAnalysis_Music}]
}
